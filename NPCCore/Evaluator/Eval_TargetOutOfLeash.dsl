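; Behavior tree evaluator for NPCs that passes its starting conditions once the target has been outside the leash for long enough.
; It reads the target-inside-leash bool and the leash exit time from blackboard keys held in variables.
; It returns true when the bool is false and the absolute gap between exit time and server world real time exceeds DelayTimeSinceLeashExit.
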(fn EvaluateStartingConditions
  (params (OwnerComp :type Behavior%20Tree%20Component%20Object%20Reference))
  (return (and (not (Blackboard|GetBool (Variables|Default|GetTargetInsideLeash) OwnerComp)) (> (Math|Float|Absolute(Float) (- (Blackboard|GetFloat (Variables|Default|GetTargetExitLeashTime) OwnerComp) (Fortnite|Sync|Time|GetServerWorldRealTimeSeconds))) (Variables|Default|GetDelayTimeSinceLeashExit)))))
