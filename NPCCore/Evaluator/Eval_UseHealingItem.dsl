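; NPC behavior tree evaluator that gates use of a healing item on whether the pawn is missing health or shield.
; On enter, it resolves the controlled FortPawn from the tree component's owning controller and caches it.
; Starting and ongoing conditions both pass while health or shield is below its maximum.
; The ongoing check calls its parent but ignores the result, and neither check looks at item availability.
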(event EventOnEnter
  (params (OwnerComp :type Behavior%20Tree%20Component%20Object%20Reference))
  (bind _ascontroller (Utilities|Casting|CastToController (Components|GetOwner OwnerComp)))
  (bind _asfort_pawn (Utilities|Casting|CastToFortPawn (Pawn|GetControlledPawn _ascontroller)))
  (Variables|Default|SetPawn _asfort_pawn))

(fn EvaluateOngoingConditions
  (params (OwnerComp :type Behavior%20Tree%20Component%20Object%20Reference))
  (|CallParentFunction OwnerComp)
  (return (or (< (FortPawn|GetHealth (Variables|Default|GetPawn)) (FortPawn|GetMaxHealth (Variables|Default|GetPawn))) (< (FortPawn|GetShield (Variables|Default|GetPawn)) (FortPawn|GetMaxShield (Variables|Default|GetPawn))))))

(fn EvaluateStartingConditions
  (params (OwnerComp :type Behavior%20Tree%20Component%20Object%20Reference))
  (return (or (< (FortPawn|GetHealth (Variables|Default|GetPawn)) (FortPawn|GetMaxHealth (Variables|Default|GetPawn))) (< (FortPawn|GetShield (Variables|Default|GetPawn)) (FortPawn|GetMaxShield (Variables|Default|GetPawn))))))
